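# Level Sequence in the ChimneyStack large male body's IK retarget folder that plays a reel of emote animations on two mannequins.
# FortPlayerMannequin and FortRetargetMannequin share a class and stand side by side with matching facing, each with transform, spawn and skeletal animation tracks.
# Both play ten back-to-back animation sections on identical frame ranges, which lines up source and retarget poses for comparison.
# It has no camera or other master tracks.
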LevelSequence /BRCosmetics/Characters/Player/Male/Large/Bodies/M_LRG_ChimneyStack/Ik_Retarget/ChimneyStack_Emote_LevelSeq
bindings [
 FortPlayerMannequin {
  id=#1
  possessed_class=FortPlayerMannequin
  tracks [
   {
    type=MovieScene3DTransformTrack
    sections [{
      type=MovieScene3DTransformSection
      channels [
       Location.X {type=MovieSceneScriptingDoubleChannel, default=-3707.1861}
       Location.Y {type=MovieSceneScriptingDoubleChannel, default=694.1856}
       Location.Z {type=MovieSceneScriptingDoubleChannel, default=1930}
       Rotation.X {type=MovieSceneScriptingDoubleChannel, default=0}
       Rotation.Y {type=MovieSceneScriptingDoubleChannel, default=0}
       Rotation.Z {type=MovieSceneScriptingDoubleChannel, default=-91.1432}
       Scale.X {type=MovieSceneScriptingDoubleChannel, default=1}
       Scale.Y {type=MovieSceneScriptingDoubleChannel, default=1}
       Scale.Z {type=MovieSceneScriptingDoubleChannel, default=1}
      ]
     }]
   }
   {
    type=MovieSceneSkeletalAnimationTrack
    sections [
     {type=MovieSceneSkeletalAnimationSection, end_frame=88, start_frame=0}
     {type=MovieSceneSkeletalAnimationSection, end_frame=456, start_frame=88}
     {type=MovieSceneSkeletalAnimationSection, end_frame=680, start_frame=456}
     {type=MovieSceneSkeletalAnimationSection, end_frame=910, start_frame=680}
     {type=MovieSceneSkeletalAnimationSection, end_frame=1198, start_frame=910}
     {type=MovieSceneSkeletalAnimationSection, end_frame=1387, start_frame=1198}
     {type=MovieSceneSkeletalAnimationSection, end_frame=1534, start_frame=1387}
     {type=MovieSceneSkeletalAnimationSection, end_frame=1783, start_frame=1534}
     {type=MovieSceneSkeletalAnimationSection, end_frame=2043, start_frame=1783}
     {type=MovieSceneSkeletalAnimationSection, end_frame=2325, start_frame=2043}
    ]
   }
   {type=MovieSceneSpawnTrack, sections [{type=MovieSceneSpawnSection, channels [None {type=MovieSceneScriptingBoolChannel, default=true}]}]}
  ]
 }
 FortRetargetMannequin {
  id=#2
  possessed_class=FortPlayerMannequin
  tracks [
   {
    type=MovieScene3DTransformTrack
    sections [{
      type=MovieScene3DTransformSection
      channels [
       Location.X {type=MovieSceneScriptingDoubleChannel, default=-3710.3611}
       Location.Y {type=MovieSceneScriptingDoubleChannel, default=535.075}
       Location.Z {type=MovieSceneScriptingDoubleChannel, default=1930}
       Rotation.X {type=MovieSceneScriptingDoubleChannel, default=0}
       Rotation.Y {type=MovieSceneScriptingDoubleChannel, default=0}
       Rotation.Z {type=MovieSceneScriptingDoubleChannel, default=-91.1432}
       Scale.X {type=MovieSceneScriptingDoubleChannel, default=1}
       Scale.Y {type=MovieSceneScriptingDoubleChannel, default=1}
       Scale.Z {type=MovieSceneScriptingDoubleChannel, default=1}
      ]
     }]
   }
   {
    type=MovieSceneSkeletalAnimationTrack
    sections [
     {type=MovieSceneSkeletalAnimationSection, end_frame=88, start_frame=0}
     {type=MovieSceneSkeletalAnimationSection, end_frame=456, start_frame=88}
     {type=MovieSceneSkeletalAnimationSection, end_frame=680, start_frame=456}
     {type=MovieSceneSkeletalAnimationSection, end_frame=910, start_frame=680}
     {type=MovieSceneSkeletalAnimationSection, end_frame=1198, start_frame=910}
     {type=MovieSceneSkeletalAnimationSection, end_frame=1387, start_frame=1198}
     {type=MovieSceneSkeletalAnimationSection, end_frame=1534, start_frame=1387}
     {type=MovieSceneSkeletalAnimationSection, end_frame=1783, start_frame=1534}
     {type=MovieSceneSkeletalAnimationSection, end_frame=2043, start_frame=1783}
     {type=MovieSceneSkeletalAnimationSection, end_frame=2325, start_frame=2043}
    ]
   }
   {type=MovieSceneSpawnTrack, sections [{type=MovieSceneSpawnSection, channels [None {type=MovieSceneScriptingBoolChannel, default=true}]}]}
  ]
 }
]
display_rate {denominator=1, numerator=30}
tick_resolution {denominator=1, numerator=24000}
tracks []
